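# Sound cue for the clap foley of the Chorus emote in Athena.
# A Random node picks one of five clap wave variations with equal weight.
# It uses the emote attenuation and emote music sound class, and sends to the music audio-analysis submix.
# A concurrency override limits simultaneous instances, and the cue keeps playing virtualized when silent.
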SoundCue /Game/Athena/Sounds/Emotes/Chorus/EmoteFoley_Chorus_Claps
root=Random_0
nodes [
 {id=Random_0, children [WavePlayer_0,WavePlayer_1,WavePlayer_2,WavePlayer_3,WavePlayer_4], properties {weights [1,1,1,1,1]}}
 {id=WavePlayer_0, properties {soundWaveAssetPtr=&/Emote_Chorus_Clap1_SFX}}
 {id=WavePlayer_1, properties {soundWaveAssetPtr=&/Emote_Chorus_Clap2_SFX}}
 {id=WavePlayer_2, properties {soundWaveAssetPtr=&/Emote_Chorus_Clap3_SFX}}
 {id=WavePlayer_3, properties {soundWaveAssetPtr=&/Emote_Chorus_Clap4_SFX}}
 {id=WavePlayer_4, properties {soundWaveAssetPtr=&/Emote_Chorus_Clap5_SFX}}
]
properties {
 attenuationSettings=&Fort_Emotes_Athena_Att
 bOverrideConcurrency=true
 maxDistance=1600
 soundClassObject=&Fort_Music_Emote
 virtualizationMode=PlayWhenSilent
 volumeMultiplier=1
 concurrencyOverrides {maxCount=3}
 soundSubmixSends [{minSendLevel=0, soundSubmix=&Fortnite_AudioAnalysis_Music}]
}
